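# Modular Control Rig for the WhineDummeez emote prop in CosmeticCompanions, a small articulated arm carrying a light.
# The skeleton chains three arm bones into a light base and mid, which branches to top and bulb, with a socket on the base.
# Connectors indicate Root and Arm chain modules plus an add-control module for each light part.
# The rig's own graph holds only an empty execution entry, and its function library is empty.
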ControlRigBlueprint /CosmeticCompanions/Assets/Props/Emotes/WhineDummeez/M_WhineDummeez
graphs [
 RigVMFunctionLibrary {links [], nodes []}
 RigVMModel {nodes [RigUnit_BeginExecution = &RigUnit_BeginExecution()]}
]
hierarchy [
 root {loc=<0 0 0>, rot=<0 0 0 1>}
 arm_a {parent=root, loc=<5.0717 -4.48 -3.0648>, rot=<0 0 0 1>}
 arm_b {parent=arm_a, loc=<0 0 3.3414>, rot=<0 0 0 1>}
 arm_c {parent=arm_b, loc=<0 4.4214 0>, rot=<0 0 0 1>}
 light_base {parent=arm_c, loc=<-5.0717 0.0586 -0.2766>, rot=<0 0 0 1>}
 light_mid {parent=light_base, loc=<-0.0678 -0.0076 2.9451>, rot=<0 0 0 1>}
 light_top {parent=light_mid, loc=<0.0678 0.0076 5.0936>, rot=<0 0 0 1>}
 light_bulb {parent=light_mid, loc=<0 0 0>, rot=<0 0 0 1>}
 light_base_socket {parent=light_base, type=Socket, loc=<0 0 0>, rot=<0 0 0 1>}
 Root/Root {type=Connector, loc=<0 0 0>, rot=<0 0 0 1>}
 Arm/Chain Primary {type=Connector, loc=<0 0 0>, rot=<0 0 0 1>}
 Arm/Start Bone {type=Connector, loc=<0 0 0>, rot=<0 0 0 1>}
 Arm/End Bone {type=Connector, loc=<0 0 0>, rot=<0 0 0 1>}
 Arm/Parent Control {type=Connector, loc=<0 0 0>, rot=<0 0 0 1>}
 Arm/Settings Control {type=Connector, loc=<0 0 0>, rot=<0 0 0 1>}
 Light Base/Add Control Primary {type=Connector, loc=<0 0 0>, rot=<0 0 0 1>}
 Light Base/Parent Control {type=Connector, loc=<0 0 0>, rot=<0 0 0 1>}
 Light Mid/Add Control Primary {type=Connector, loc=<0 0 0>, rot=<0 0 0 1>}
 Light Mid/Parent Control {type=Connector, loc=<0 0 0>, rot=<0 0 0 1>}
 Light Top/Add Control Primary {type=Connector, loc=<0 0 0>, rot=<0 0 0 1>}
 Light Top/Parent Control {type=Connector, loc=<0 0 0>, rot=<0 0 0 1>}
 Light Bulb/Add Control Primary {type=Connector, loc=<0 0 0>, rot=<0 0 0 1>}
 Light Bulb/Parent Control {type=Connector, loc=<0 0 0>, rot=<0 0 0 1>}
]
